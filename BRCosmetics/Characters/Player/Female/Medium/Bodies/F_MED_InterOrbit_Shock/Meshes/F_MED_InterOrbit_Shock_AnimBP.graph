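# Post-pose Animation Blueprint for the female InterOrbit Shock body, layering a Control Rig on the pose copied from the attached parent mesh.
# Each update converts movement, targeting, parachute and emote states into rig inputs.
# Weapon changes read the owner's FortPlayerAnimInstance to set pistol, rifle or melee weights; melee never reaches the rig.
# Only Crouch ORs crouching with not-jumping and not-falling, so it is also true outside crouch.
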AnimBlueprint /BRCosmetics/Characters/Player/Female/Medium/Bodies/F_MED_InterOrbit_Shock/Meshes/F_MED_InterOrbit_Shock_AnimBP
target_skeleton=&/F_MED_InterOrbit_Shock_Skeleton
graphs [
 AnimGraph {nodes [
   n1 = Control Rig!(
    Source=Copy Pose From Mesh() {bUseAttachedParent=true}
    Alpha
    Crouch:double=Get Crouch() -> (Crouch:double)
    CrouchBool:bool=Get Only Crouch() -> (Only Crouch:bool)
    Parachute:double=Get Parachute Open() -> (Parachute Open:double)
    Emote:double=Get Emote() -> (Emote:double)
    Pistol:double=Get pistol() -> (pistol:double)
    Targeting:double=Get Targeting() -> (Targeting:double)
    Rifel:double=Get Rifle() -> (Rifle:double)
    Jump:double=Get jump() -> (jump:double)
    SkydiveFloating:double=Get Skydive Floating() -> (Skydive Floating:double)
   ) -> (Pose) {controlRigAssetReference {blueprintRigClass=&/F_MED_InterOrbit_Shock_CtrlRig.F_MED_InterOrbit_Shock_CtrlRig_C}}
   Output Pose(Result=@n1) {layerGroup=None}
  ]}
 EventGraph {nodes [
   Event BlueprintUpdateAnimation()
   > Set Crouch(Crouch:double="To Float (Boolean)"(InBool=Get bIsCrouching())) -> (Output_Get:double)
   > Set Parachute Open(Parachute Open:double="To Float (Boolean)"(InBool=Get bIsParachuteOpened())) -> (Output_Get:double)
   > Set Emote(Emote:double="To Float (Boolean)"(InBool=Get bIsPlayingEmote())) -> (Output_Get:double)
   > Set Targeting(Targeting:double="To Float (Boolean)"(InBool=Get bIsTargeting())) -> (Output_Get:double)
   > Set Skydive Floating(Skydive Floating:double="To Float (Boolean)"(InBool=Get bIsSkydiveFloating())) -> (Output_Get:double)
   n1 = Get bIsFalling()
   n2 = Get bIsJumping()
   > Set Only Crouch(Only Crouch:bool=OR Boolean!(A=Get bIsCrouching(), B=NOT Boolean(A=@n2), C:bool=NOT Boolean(A=@n1)) -> (ReturnValue)) -> (Output_Get:bool)
   n3 = Get bIsJumping()
   n4 = Get bIsFalling()
   > Set jump(jump:double="To Float (Boolean)"(InBool=OR Boolean!(A=@n3, B=@n4) -> (ReturnValue))) -> (Output_Get:double)
   Event OnWeaponChanged()
   > n5 = Cast To FortPlayerAnimInstance(Object=GetAnimInstance(self=GetOwnerSkeletalMeshComponent()))
   > Set pistol(pistol:double=0) -> (Output_Get:double)
   > Set Melee(Melee:double=0) -> (Output_Get:double)
   > Set Rifle(Rifle:double=0) -> (Output_Get:double)
   > n6 = Switch on EFortWeaponCoreAnimation!(execute, Selection:EFortWeaponCoreAnimation Enum=Get WeaponCoreAnim(self=@n5)) -> (
    Melee:Exec
    Pistol:Exec
    Shotgun:Exec
    PaperBlueprint:Exec
    Rifle:Exec
    MeleeOneHand:Exec
    MachinePistol:Exec
    RocketLauncher:Exec
    GrenadeLauncher:Exec
    GoingCommando:Exec
    AssaultRifle:Exec
    TacticalShotgun:Exec
    SniperRifle:Exec
    TrapPlacement:Exec
    ShoulderLauncher:Exec
    AbilityDecoTool:Exec
    Crossbow:Exec
    C4:Exec
    RemoteControl:Exec
    DualWield:Exec
    AR_BullPup:Exec
    AR_ForwardGrip:Exec
    MedPackPaddles:Exec
    SMG_P90:Exec
    AR_DrumGun:Exec
    Consumable_Small:Exec
    Consumable_Large:Exec
    Balloon:Exec
    MountedTurret:Exec
    CreativeTool:Exec
    ExplosiveBow:Exec
    AshtonIndigo:Exec
    AshtonChicago:Exec
    MeleeDualWield:Exec
    Unarmed:Exec
   )
   Set Rifle(
    execute=[
     @n6.AR_BullPup
     @n6.AR_DrumGun
     @n6.AR_ForwardGrip
     @n6.AssaultRifle
     @n6.Crossbow
     @n6.GrenadeLauncher
     @n6.Rifle
     @n6.RocketLauncher
     @n6.SMG_P90
     @n6.Shotgun
     @n6.SniperRifle
     @n6.TacticalShotgun
    ]
    Rifle=1
   )
   Set pistol(execute=[@n6.MachinePistol, @n6.Pistol], pistol=1)
   Set Melee(execute=@n6.Melee, Melee=1)
  ]}
]
variables [
 Crouch {type=number}
 Emote {type=number}
 jump {type=number}
 Melee {type=number}
 Only Crouch {type=boolean}
 Parachute Open {type=number}
 pistol {type=number}
 Rifle {type=number}
 Skydive Floating {type=number}
 Targeting {type=number}
]
