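# Physics asset for the face accessory part of the female InterOrbit Shock cosmetic body, covering a short left-side hair chain.
# Single capsules sit on the dynamic hair bone, its base, the head, the neck and the right upper arm.
# Tight swing limits and zero twist chain the hair to the head, whose joint to the neck rotates freely.
# The upper-arm capsule has no constraint.
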PhysicsAsset /BRCosmetics/Characters/Player/Female/Medium/Bodies/F_MED_InterOrbit_Shock/Meshes/Parts/F_MED_InterOrbit_Shock_FaceAcc_Physics
bodies {
 base_hair_side_l [{shapeType=Capsule, length=0.505, radius=0.505, shapeName=base_hair_side_l_capsule, rotation=<rot 0 -90 0>}]
 dyn_hair_side_b_01_l [{shapeType=Capsule, length=5, radius=1.5, shapeName=dyn_hair_side_b_01_l_capsule, center=<2 0 0.8>, rotation=<rot 0.0001 89.9998 100>}]
 head [{shapeType=Capsule, length=6.7353, radius=7.5, center=<0.3665 2.0786 -0.7349>, rotation=<rot 0 80 90>}]
 neck_01 [{shapeType=Capsule, length=5.9069, radius=4.3349, shapeName=neck_01_capsule}]
 upperarm_r [{shapeType=Capsule, length=10, radius=7.3, shapeName=upperarm_r_capsule, center=<0.3038 -0.5163 -0.758>, rotation=<rot 0 90 90>}]
}
constraints [
 {bone1Name=dyn_hair_side_b_01_l, bone2Name=base_hair_side_l, swing1LimitDegrees=10, swing2LimitDegrees=10, twistLimitDegrees=0}
 {bone1Name=base_hair_side_l, bone2Name=head, swing1LimitDegrees=5, swing2LimitDegrees=5, twistLimitDegrees=0, twistMotion=Locked}
 {
  bone1Name=head
  bone2Name=neck_01
  swing1LimitDegrees=45
  swing1Motion=Free
  swing2LimitDegrees=45
  swing2Motion=Free
  twistLimitDegrees=45
  twistMotion=Free
 }
]
